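# Modular-style Control Rig for the egg prop in the EggBawk emote of CosmeticCompanions.
# The hierarchy has a whole-egg element plus top, bottom and three numbered shard elements, all parented directly to Root.
# Each piece has Add Control Primary and Parent Control connectors, suggesting per-piece control modules, alongside a Root connector.
# The main graph holds only an unlinked Begin Execution node, and the function library is empty.
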ControlRigBlueprint /CosmeticCompanions/Assets/Props/Emotes/EggBawk/MR_EggBawk_Egg
graphs [
 RigVMFunctionLibrary {links [], nodes []}
 RigVMModel {nodes [RigUnit_BeginExecution = &RigUnit_BeginExecution()]}
]
hierarchy [
 Root {loc=<0 0 0>, rot=<0 0 0 1>}
 whole {parent=Root, loc=<0 0 25.889>, rot=<0 0 0 1>}
 shard_top {parent=Root, loc=<0 0 48.4413>, rot=<0 0 0 1>}
 shard_bottom {parent=Root, loc=<0 0 13.6438>, rot=<0 0 0 1>}
 shard_01 {parent=Root, loc=<-7.3324 20.9766 29.2341>, rot=<0 0 0.8359 0.5488>}
 shard_02 {parent=Root, loc=<14.9403 15.4936 32.6115>, rot=<0 0 0.4266 0.9044>}
 shard_03 {parent=Root, loc=<11.8802 18.4961 29.9117>, rot=<0 0 0.5003 0.8658>}
 Root/Root {type=Connector, loc=<0 0 0>, rot=<0 0 0 1>}
 shard_bottom/Add Control Primary {type=Connector, loc=<0 0 0>, rot=<0 0 0 1>}
 shard_bottom/Parent Control {type=Connector, loc=<0 0 0>, rot=<0 0 0 1>}
 shard_01/Add Control Primary {type=Connector, loc=<0 0 0>, rot=<0 0 0 1>}
 shard_01/Parent Control {type=Connector, loc=<0 0 0>, rot=<0 0 0 1>}
 whole/Add Control Primary {type=Connector, loc=<0 0 0>, rot=<0 0 0 1>}
 whole/Parent Control {type=Connector, loc=<0 0 0>, rot=<0 0 0 1>}
 shard_03/Add Control Primary {type=Connector, loc=<0 0 0>, rot=<0 0 0 1>}
 shard_03/Parent Control {type=Connector, loc=<0 0 0>, rot=<0 0 0 1>}
 shard_02/Add Control Primary {type=Connector, loc=<0 0 0>, rot=<0 0 0 1>}
 shard_02/Parent Control {type=Connector, loc=<0 0 0>, rot=<0 0 0 1>}
 shard_top/Add Control Primary {type=Connector, loc=<0 0 0>, rot=<0 0 0 1>}
 shard_top/Parent Control {type=Connector, loc=<0 0 0>, rot=<0 0 0 1>}
]
