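# Control Rig for the egg mountain prop of the EggBawk cosmetic companion emote.
# The bone chain runs from Root through mountain to mountain_mid, which parents three egg bones.
# Connector names point to a modular rig, with a Root module and Add Control Primary and Parent Control connectors for each mountain and egg part.
# Its own graph holds only an empty BeginExecution event, with no logic or links.
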ControlRigBlueprint /CosmeticCompanions/Assets/Props/Emotes/EggBawk/MR_EggBawk_EggMountain
graphs [
 RigVMFunctionLibrary {links [], nodes []}
 RigVMModel {nodes [RigUnit_BeginExecution = &RigUnit_BeginExecution()]}
]
hierarchy [
 Root {loc=<-0.9535 -0.3353 0>, rot=<0 0 0 1>}
 mountain {parent=Root, loc=<0 0 0>, rot=<0 -0.0229 0 0.9997>}
 mountain_mid {parent=mountain, loc=<0 0 31.2658>, rot=<0 0 0 1>}
 egg_01 {parent=mountain_mid, loc=<-2.9446 0.0912 20.7998>, rot=<-0.7441 0.2763 -0.5774 -0.1913>}
 egg_02 {parent=mountain_mid, loc=<-4.4542 -1.3495 28.2803>, rot=<0.3003 -0.157 0.8757 0.344>}
 egg_03 {parent=mountain_mid, loc=<-1.023 -0.0173 37.2043>, rot=<-0.1321 -0.0551 -0.926 0.3494>}
 Root/Root {type=Connector, loc=<0 0 0>, rot=<0 0 0 1>}
 mountain/Add Control Primary {type=Connector, loc=<0 0 0>, rot=<0 0 0 1>}
 mountain/Parent Control {type=Connector, loc=<0 0 0>, rot=<0 0 0 1>}
 mountain mid/Add Control Primary {type=Connector, loc=<0 0 0>, rot=<0 0 0 1>}
 mountain mid/Parent Control {type=Connector, loc=<0 0 0>, rot=<0 0 0 1>}
 egg 1/Add Control Primary {type=Connector, loc=<0 0 0>, rot=<0 0 0 1>}
 egg 1/Parent Control {type=Connector, loc=<0 0 0>, rot=<0 0 0 1>}
 egg 2/Add Control Primary {type=Connector, loc=<0 0 0>, rot=<0 0 0 1>}
 egg 2/Parent Control {type=Connector, loc=<0 0 0>, rot=<0 0 0 1>}
 egg 3/Add Control Primary {type=Connector, loc=<0 0 0>, rot=<0 0 0 1>}
 egg 3/Parent Control {type=Connector, loc=<0 0 0>, rot=<0 0 0 1>}
]
