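# PA_EggBawk_Egg is the physics asset for the egg prop mesh used by the EggBawk cosmetic emote.
# It holds a single body on the Root bone, shaped as a zero-length capsule that acts as an offset sphere around the egg.
# It has no constraints, so it defines no jointed ragdoll setup.
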PhysicsAsset /CosmeticCompanions/Assets/Props/Emotes/EggBawk/Meshes/PA_EggBawk_Egg
bodies {Root [{shapeType=Capsule, length=0, radius=46.886, center=<-1.5338 16.6893 37.6611>, rotation=<rot 0 -86.8515 -89.753>}]}
constraints []
