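# Physics asset for the egg mountain skeletal mesh prop in the EggBawk emote of the CosmeticCompanions content.
# It gives four bones one capsule body each: three small, similarly sized capsules for the individual eggs and one much larger capsule for the mountain.
# It defines no constraints, so no joints link the bodies to one another.
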PhysicsAsset /CosmeticCompanions/Assets/Props/Emotes/EggBawk/Meshes/PA_EggBawk_EggMountain
bodies {
 egg_01 [{shapeType=Capsule, length=6.0776, radius=4.5433, shapeName=egg_01_capsule, center=<0.0142 -0.0275 1.817>, rotation=<rot -0.8672 -90 0.4465>}]
 egg_02 [{shapeType=Capsule, length=6.0776, radius=4.7645, shapeName=egg_02_capsule, center=<0.0208 -0.0229 1.8161>, rotation=<rot -0.7233 -90 0.6544>}]
 egg_03 [{
   shapeType=Capsule
   length=6.0776
   radius=4.7534
   shapeName=egg_03_capsule
   center=<-0.0199 -0.0236 1.8161>
   rotation=<rot -0.7456 -90 -0.629>
  }]
 mountain [{shapeType=Capsule, length=25.0169, radius=21.8571, center=<-0.0753 -0.28 24.5404>, rotation=<rot -0.7757 -90 -0.2139>}]
}
constraints []
